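# Niagara system for the Dazzle skydiving contrail cosmetic, with toon stars, arrows, squiggles, rings, ribbons and a TigerVisualiser mesh attached to the character's skeletal mesh.
# System-level dive and velocity values orient and stretch meshes, push ribbons and drive material parameters.
# Stars trail their sampled bones by smoothing position toward velocity-based offsets with InterpolateOverTime.
# Front-end switches substitute fixed directions and alternate scales for menu previews.
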NiagaraSystem /BRCosmetics/Athena/Effects/Particles/Skydiving/NS_Contrail_Dazzle/FX/NS_Skydive_Dazzle_v4
dependencies {
 usedDataInterfaces [&NiagaraDataInterfaceArrayInt32, &NiagaraDataInterfaceCurve, &NiagaraDataInterfaceArrayFloat]
 usedDynamicInputs [&SetBoolByFloatComparison]
 usedModules [
  &SystemState
  /Game/Effects/Niagara/Modules/Skydiving/FN_DivingAttributes/v5/FN_DivingAttributes
  /Game/Effects/Niagara/Modules/Skydiving/FN_DivingAttributes/v6/FN_DivingAttributes
  &self:SystemSpawnScript.SystemScriptSource.SystemScriptGraph.NiagaraNodeAssignment_2.SetVariables_#1
  &self:SystemSpawnScript.SystemScriptSource.SystemScriptGraph.NiagaraNodeAssignment_1.SetVariables_#2
  &FN_Integrator
  &self:SystemSpawnScript.SystemScriptSource.SystemScriptGraph.NiagaraNodeAssignment_0.SetVariables_#3
  &self:SystemSpawnScript.SystemScriptSource.SystemScriptGraph.NiagaraNodeAssignment_7.SetVariables_#4
  &Timeline
 ]
}
emitters {
 Arrow {
  emitterUpdateScript [
   &EmitterState(Delay First Loop Only=true, Life Cycle Mode=::NewEnumerator1, Loop Delay=0.2, Loop Duration=1, UseLoopDelay=true)
   &SpawnRate(Spawn Probability=0.44, SpawnRate=10, Use Spawn Probability=true)
  ]
  particleSpawnScript [
   &FN_SetSpawnGroup()
   &InitializeParticle(
    Color Mode=::NewEnumerator0
    Lifetime=&Multiply_Float(
     A=&RandomRangeFloat(Minimum=0.75, Maximum=1)
     B=&MakeCustomFloatFromBool(
      Bool=&SetBoolByIntComparison(A=2, B=$Particles.SpawnGroup, Comparison Type=::NewEnumerator0)
      True Float=1
      False Float=0.5
     )
    )
    Mesh Scale=&Multiply_VectorByFloat(
     Vector=&RandomRangeVector(Minimum=<1 1 2.25>, Maximum=<1 1 3>)
     Float=&MakeCustomFloatFromBool(
      Bool=&SetBoolByIntComparison(A=2, B=$Particles.SpawnGroup, Comparison Type=::NewEnumerator0)
      True Float=&FN_FrontEndFloatModifier(Gameplay=1, Front End=0.9)
      False Float=&FN_FrontEndFloatModifier(Gameplay=0.5, Front End=0.3)
     )
    )
    Mesh Scale Mode=::NewEnumerator4
    UsePositionOffset=false
    Write Material Random=true
   )
   &InitialMeshOrientation(
    Mesh Orientation Mode=::NewEnumerator2
    Rotation=&RandomRangeVector(Minimum=<0 0 0>, Maximum=<0 0 1>)
    Use Rotation Vector=true
   )
  ]
  particleUpdateScript [
   &ParticleState()
   &ScaleMeshSize(Scale Factor=&MakeCustomVectorFromBool(
     Bool=$User.bIsFrontEndPreview
     True Vector=&Lerp_Vector(A=<1 1 1.1>, B=<1 1 1.3>, Alpha=$System.Dive)
     False Vector=&Lerp_Vector(A=&RandomRangeVector(Minimum=<1 1 1>, Maximum=<1 1 1.1>), B=<1 1 1.3>, Alpha=$System.Dive)
    ))
   &SkeletalMeshLocation(
    Bone / Socket Index=&SelectIntFromArray(Direct Array Index=$Particles.SpawnGroup, Int Selection Array=DataInterface() {propertyValues {IntData [0,1,2,3]}})
    Bone Sampling Mode=::NewEnumerator4
    MaskSpawnGroup=false
    Orientation Sampling=::NewEnumerator1
    Sampled Position Offset Coordinate Space:ENiagaraSkelMeshTransforms=::NewEnumerator1
    Skeletal Mesh=$User.SkeletalMesh
   )
   &DynamicMaterialParameters(
    Index 0 Param 1=&FloatFromCurve(FloatCurve=DataInterface() {propertyValues {Curve {keys [{tangentMode=RCTM_User, time=0, value=1}, {tangentMode=RCTM_User, time=1, value=0}]}}})
    Index 0 Param 2=&SelectFloatFromArray(Direct Array Index=$Particles.SpawnGroup, Float Selection Array=DataInterface() {propertyValues {FloatData [9,9,4,4]}})
    Index 0 Param 3=&SelectFloatFromArray(Direct Array Index=$Particles.SpawnGroup, Float Selection Array=DataInterface() {propertyValues {FloatData [6,6,4,4]}})
    Index 0 Param 4=&SelectFloatFromArray(Direct Array Index=$Particles.SpawnGroup, Float Selection Array=DataInterface() {propertyValues {FloatData [12,12,5,5]}})
    Module.Param0WriteEnabled=true
    Module.Param0WriteEnabled1=false
    Module.Param0WriteEnabled2=true
    Module.Param0WriteEnabled3=true
    Module.Param1WriteEnabled=true
    Module.Param1WriteEnabled1=false
    Module.Param1WriteEnabled2=true
    Module.Param1WriteEnabled3=true
    Module.Param2WriteEnabled=true
    Module.Param2WriteEnabled1=false
    Module.Param2WriteEnabled2=true
    Module.Param2WriteEnabled3=true
    Module.Param3WriteEnabled=true
    Module.Param3WriteEnabled1=false
    Module.Param3WriteEnabled2=true
    Module.Param3WriteEnabled3=true
    Write Parameter Index 1=true
    Write Parameter Index 2=true
    Write Parameter Index 3=true
   )
   &Color(Color=<rgba 0.1521 1 0.3289 1>)
   /SetVariables(Particles.NewMeshOrientation:Quat4f=/Niagara/DynamicInputs/TypeConversions/MakeCustomQuaternionFromBool(
     Bool:NiagaraBool=$User.bIsFrontEndPreview
     True Quat:Quat4f=/BRCosmetics/Athena/Effects/Particles/Skydiving/NS_Contrail_Apprentice/OrientMestToVectorFromCustomRot(
      InitialRotation:Quat4f=$Particles.MeshOrientation
      Look At Direction:Vector3f=&TransformVector(Vector=&Lerp_Vector(A=<0 0 -1>, B=<0 0.6 -0.5>, Alpha=$System.Dive), Destination Space=::Simulation)
      Reference Axis Vector:Vector3f=&TransformVector(Vector=<0 0 -1>, Destination Space=::Simulation)
      RotationVector:Vector3f=<0 0 0>
     )
     False Quat:Quat4f=&OrientMestToVectorFromCustomRot(
      InitialRotation:Quat4f=$Particles.MeshOrientation
      Look At Direction:Vector3f=&TransformVector(Vector=&Lerp_Vector(A=$System.SkydiveVelocity, B=<0 0 -1>, Alpha=$System.Dive), Destination Space=::Simulation)
      Reference Axis Vector:Vector3f=&TransformVector(Vector=<0 0 -1>, Destination Space=::Simulation)
      RotationVector:Vector3f=<0 0 0>
     )
    )) {isSetParameters=true, module=SetVariables_#5}
  ]
  rendererClasses [&NiagaraMeshRendererProperties]
  renderers [{rendererClass=&NiagaraMeshRendererProperties, rendererIndex=0}]
 }
 Ribbons {
  emitterUpdateScript [
   &EmitterState(Inactive Response=::NewEnumerator1, Life Cycle Mode=::NewEnumerator1, Loop Duration=0, UseLoopDelay=false)
   /Game/Effects/Niagara/Modules/Skydiving/FN_DivingAttributes/FN_EmitterDivingAttributes_v1(Bone Index=1, SkeletalMesh:NiagaraDataInterfaceSkeletalMesh=$User.SkeletalMesh)
   &FN_SpawnRibbon_v2(Number of Ribbons=2, Spawn Mode=::NewEnumerator0, SpawnRate=128)
   /SetVariables(Emitter.TrailFade:NiagaraFloat=&FloatFromCurve(
     FloatCurve=DataInterface() {propertyValues {Curve {keys [{interpMode=RCIM_Cubic, tangentMode=RCTM_User, time=0, value=0}, {interpMode=RCIM_Cubic, tangentMode=RCTM_User, time=0.75, value=1}]}}}
     CurveIndex=$Emitter.Age
    )) {isSetParameters=true}
  ]
  particleSpawnScript [
   &PartitionParticles()
   &FN_InitializeRibbon_v3(Custom Offset=true, Enable Skeletal Position=true, Lifetime=0.15, Local Vector Offset:Vector3f=<0 1 0>, Ribbon Width=65)
   &AddVelocity(
    Scale Added Velocity=&Lerp_Vector(A=<1 1 1>, B=<4 4 1>, Alpha=$System.Dive)
    Velocity=&FN_FrontEndVector3Modifier(
     Gameplay=&Multiply_Vector(
      A=&NormalizeVector(Vector To Normalize=$System.SkydiveVelocity, Safe Normalize=true, Safe Fallback Threshold=0, Safe Fallback Vector=<0 0 0.001>)
      B=<1 1 -1500>
     )
     Front End=&Lerp_Vector(A=<0 0 1500>, B=<1 -300 1000>, Alpha=$System.Dive)
    )
   )
   &DynamicMaterialParameters(
    Index 0 Param 1=&Multiply_Float(
     A=&SelectFloatFromArray(
      Direct Array Index=$Particles.PartitionParticles.Partition
      Float Selection Array=DataInterface() {propertyValues {FloatData [2.55,2.22,1.55,1.77]}}
     )
     B=1
    )
    Index 0 Param 2=&Multiply_Float(
     A=&SelectFloatFromArray(Direct Array Index=$Particles.PartitionParticles.Partition, Float Selection Array=DataInterface() {propertyValues {FloatData [2.33,2.11]}})
     B=1
    )
    Index 0 Param 3=&SelectFloatFromArray(Direct Array Index=$Particles.PartitionParticles.Partition, Float Selection Array=DataInterface() {propertyValues {FloatData [2.5,2.4]}})
    Index 0 Param 4=&SelectFloatFromArray(Direct Array Index=$Particles.PartitionParticles.Partition, Float Selection Array=DataInterface() {propertyValues {FloatData [0,0.44]}})
    Index 1 Param 1=0.1
    Index 1 Param 2=0.1
    Index 1 Param 3=0.25
    Index 2 Param 1=0.7
    Index 2 Param 2=&SelectFloatFromArray(
     Direct Array Index=$Particles.PartitionParticles.Partition
     Float Selection Array=DataInterface() {propertyValues {FloatData [0.25,0.25,0.15,0.15]}}
    )
    Module.Param0WriteEnabled=false
    Module.Param0WriteEnabled1=true
    Module.Param0WriteEnabled2=true
    Module.Param1WriteEnabled=false
    Module.Param1WriteEnabled1=true
    Module.Param1WriteEnabled2=true
    Module.Param2WriteEnabled=true
    Module.Param2WriteEnabled1=true
    Module.Param2WriteEnabled2=true
    Module.Param3WriteEnabled=true
    Module.Param3WriteEnabled1=true
    Module.Param3WriteEnabled2=true
    Write Parameter Index 1=true
    Write Parameter Index 2=true
   )
   &SkeletalMeshLocation(
    Apply Sampled Orientation to Mesh Particles=true
    Bone / Socket Index=$Particles.PartitionParticles.Partition
    Bone Sampling Mode=::NewEnumerator4
    Inherit Velocity=false
    Interpolate Mesh Sampling=false
    MaskSpawnGroup=false
    Sampled Position Offset Coordinate Space:ENiagaraMeshTransforms=::NewEnumerator3
    Skeletal Mesh=$User.SkeletalMesh
   )
  ]
  particleUpdateScript [
   &InheritSourceMovement(Applied Force Scale=<0.5 0.5 1>, Applied Position Scale=<0.5 0.5 1>, ApplyForce=false, ApplyPosition=true, ApplyPositionFalloff=false)
   &DynamicMaterialParameters(
    Index 0 Param 1=&Multiply_Float(
     A=&SelectFloatFromArray(
      Direct Array Index=$Particles.PartitionParticles.Partition
      Float Selection Array=DataInterface() {propertyValues {FloatData [2.55,2.22,1.55,1.77]}}
     )
     B=&Lerp_Float(A=1, B=0.5, Alpha=$System.FN_Integrator001.Value)
    )
    Index 0 Param 2=&Multiply_Float(
     A=&SelectFloatFromArray(
      Direct Array Index=$Particles.PartitionParticles.Partition
      Float Selection Array=DataInterface() {propertyValues {FloatData [2.33,2.11,1.55,1.66]}}
     )
     B=&Lerp_Float(A=1, B=0.5, Alpha=$System.FN_Integrator001.Value)
    )
    Index 0 Param 3=&SelectFloatFromArray(
     Direct Array Index=$Particles.PartitionParticles.Partition
     Float Selection Array=DataInterface() {propertyValues {FloatData [2.5,2.4,3,2.8]}}
    )
    Index 0 Param 4=&SelectFloatFromArray(
     Direct Array Index=$Particles.PartitionParticles.Partition
     Float Selection Array=DataInterface() {propertyValues {FloatData [0,0.44,0,0.44]}}
    )
    Module.Param0WriteEnabled=true
    Module.Param1WriteEnabled=true
    Module.Param2WriteEnabled=false
    Module.Param3WriteEnabled=false
   ) {module=DynamicMaterialParameters001}
   &SolveForcesAndVelocity(Clamp Velocity=false, Limit Acceleration=false)
   &FN_UpdateRibbon(Detach Locked Particle=true, Enable Skeletal Position=true)
  ]
  rendererClasses [&NiagaraRibbonRendererProperties]
  renderers [{rendererClass=&NiagaraRibbonRendererProperties}]
 }
 Ring {
  emitterUpdateScript [
   &EmitterState(Life Cycle Mode=::NewEnumerator1, Loop Behavior=::NewEnumerator1, Loop Duration Mode=::NewEnumerator1, UseLoopDelay=false)
   &SpawnBurst_Instantaneous(Spawn Count=2, Use Spawn Probability=false, UseLoopCountLimit=false)
  ]
  particleSpawnScript [
   &PartitionParticles()
   &InitializeParticle(Mesh Scale Mode=::NewEnumerator3, UsePositionOffset=false, Write Material Random=true)
   &InitialMeshOrientation(
    Mesh Orientation Mode=::NewEnumerator2
    Orientation Vector=<0 0 1>
    Rotation=&RandomRangeVector(Minimum=<0 0 0>, Maximum=<0 0 1>)
    Use Rotation Vector=true
   )
   &CameraOffset(Camera Offset Amount=5)
   /SetVariables(
    Particles.Dynamic1:Vector4f=<25 0 0 0>
    Particles.Ring1Color:LinearColor=<rgba 0.776 0.0849 0.2922 1>
    Particles.Ring1Scale:Vector3f=<0.25 0.25 0.25>
   ) {isSetParameters=true}
   /SetVariables(Particles.Dynamic2:Vector4f=<10 0 0 0>, Particles.Ring2Color:LinearColor=<rgba 0 0.45 1 1>, Particles.Ring2Scale:Vector3f=<0.3 0.3 0.3>) {isSetParameters=true}
  ]
  particleUpdateScript [
   &ParticleState(Kill Particles When Lifetime Has Elapsed=false)
   &SkeletalMeshLocation(
    Bone / Socket Index=$Particles.PartitionParticles.Partition
    Bone Sampling Mode=::NewEnumerator4
    MaskSpawnGroup=false
    Sampled Position Offset=<0 0 1>
    Sampled Position Offset Coordinate Space:ENiagaraSkelMeshTransforms=::NewEnumerator2
    Skeletal Mesh=$User.SkeletalMesh
   )
   /SetVariables(
    Particles.Ring1Scale:Vector3f=&Multiply_VectorByFloat(
     Vector=<0.25 0.25 0.25>
     Float=&FloatFromCurve(
      FloatCurve=DataInterface() {propertyValues {Curve {keys [{tangentMode=RCTM_User, time=0, value=0}, {tangentMode=RCTM_User, time=0.5, value=1}]}}}
      CurveIndex=$Emitter.Age
     )
    )
    Particles.Ring2Scale:Vector3f=&Multiply_VectorByFloat(
     Vector=<0.3 0.3 0.3>
     Float=&FloatFromCurve(
      FloatCurve=DataInterface() {propertyValues {Curve {keys [{tangentMode=RCTM_User, time=0, value=0}, {tangentMode=RCTM_User, time=0.5, value=1}]}}}
      CurveIndex=$Emitter.Age
     )
    )
   ) {isSetParameters=true}
  ]
  rendererClasses [&NiagaraMeshRendererProperties]
  renderers [
   {rendererClass=&NiagaraMeshRendererProperties, rendererIndex=0}
   {rendererClass=&NiagaraMeshRendererProperties, rendererIndex=1}
  ]
 }
 Squiggle {
  emitterUpdateScript [
   &EmitterState(Delay First Loop Only=true, Life Cycle Mode=::NewEnumerator1, Loop Delay=0.2, Loop Duration=1, UseLoopDelay=true)
   &SpawnRate(Spawn Probability=0.33, SpawnRate=6, Use Spawn Probability=true)
  ]
  particleSpawnScript [
   &FN_SetSpawnGroup(Number of Spawn Groups=2)
   &InitializeParticle(
    Color Mode=::NewEnumerator0
    Lifetime=&Multiply_Float(
     A=&RandomRangeFloat(Minimum=0.75, Maximum=1)
     B=&MakeCustomFloatFromBool(
      Bool=&SetBoolByIntComparison(A=2, B=$Particles.SpawnGroup, Comparison Type=::NewEnumerator0)
      True Float=1
      False Float=0.5
     )
    )
    Mesh Scale=&Multiply_VectorByFloat(Vector=&RandomRangeVector(Minimum=<1 1 2>, Maximum=<1 1 2.5>), Float=1)
    Mesh Scale Mode=::NewEnumerator4
    UsePositionOffset=false
    Write Material Random=true
   )
   &InitialMeshOrientation(
    Mesh Orientation Mode=::NewEnumerator2
    Rotation=&RandomRangeVector(Minimum=<0 0 0>, Maximum=<0 0 1>)
    Use Rotation Vector=true
   )
  ]
  particleUpdateScript [
   &ParticleState()
   &ScaleMeshSize(Scale Factor=&MakeCustomVectorFromBool(
     Bool=$User.bIsFrontEndPreview
     True Vector=&Lerp_Vector(A=<1 1 1.1>, B=<1 1 1.3>, Alpha=0.5)
     False Vector=&Lerp_Vector(A=&RandomRangeVector(Minimum=<1 1 0.8>, Maximum=<1 1 0.9>), B=<1 1 1.3>, Alpha=$System.Dive)
    ))
   &SkeletalMeshLocation(
    Bone / Socket Index=&SelectIntFromArray(Direct Array Index=$Particles.SpawnGroup, Int Selection Array=DataInterface() {propertyValues {IntData [0,1,2,3]}})
    Bone Sampling Mode=::NewEnumerator4
    MaskSpawnGroup=false
    Orientation Sampling=::NewEnumerator1
    Sampled Position Offset Coordinate Space:ENiagaraSkelMeshTransforms=::NewEnumerator1
    Skeletal Mesh=$User.SkeletalMesh
   )
   &DynamicMaterialParameters(
    Index 0 Param 1=&FloatFromCurve(FloatCurve=DataInterface() {propertyValues {Curve {keys [{tangentMode=RCTM_User, time=0, value=1}, {tangentMode=RCTM_User, time=1, value=0}]}}})
    Index 0 Param 2=6
    Index 0 Param 3=6
    Index 0 Param 4=&SelectFloatFromArray(Direct Array Index=$Particles.SpawnGroup, Float Selection Array=DataInterface() {propertyValues {FloatData [10,10,5,5]}})
    Module.Param0WriteEnabled=true
    Module.Param0WriteEnabled1=false
    Module.Param0WriteEnabled2=true
    Module.Param0WriteEnabled3=true
    Module.Param1WriteEnabled=true
    Module.Param1WriteEnabled1=false
    Module.Param1WriteEnabled2=true
    Module.Param1WriteEnabled3=true
    Module.Param2WriteEnabled=true
    Module.Param2WriteEnabled1=false
    Module.Param2WriteEnabled2=true
    Module.Param2WriteEnabled3=true
    Module.Param3WriteEnabled=true
    Module.Param3WriteEnabled1=false
    Module.Param3WriteEnabled2=true
    Module.Param3WriteEnabled3=true
    Write Parameter Index 1=true
    Write Parameter Index 2=true
    Write Parameter Index 3=true
   )
   &Color(Color=<rgba 0.1046 0.0439 0.3646 1>) {module=Color001}
   /SetVariables(Particles.NewMeshOrientation:Quat4f=&MakeCustomQuaternionFromBool(
     Bool:NiagaraBool=$User.bIsFrontEndPreview
     True Quat:Quat4f=&OrientMestToVectorFromCustomRot(
      InitialRotation:Quat4f=$Particles.MeshOrientation
      Look At Direction:Vector3f=&TransformVector(Vector=&Lerp_Vector(A=<0 0 -1>, B=<0 0.6 -0.5>, Alpha=$System.Dive), Destination Space=::Simulation)
      Reference Axis Vector:Vector3f=&TransformVector(Vector=<0 0 -1>, Destination Space=::Simulation)
      RotationVector:Vector3f=<0 0 0>
     )
     False Quat:Quat4f=&OrientMestToVectorFromCustomRot(
      InitialRotation:Quat4f=$Particles.MeshOrientation
      Look At Direction:Vector3f=&TransformVector(Vector=$System.SkydiveVelocity, Destination Space=::Simulation)
      Reference Axis Vector:Vector3f=&TransformVector(Vector=<0 0 -1>, Destination Space=::Simulation)
      RotationVector:Vector3f=<0 0 0>
     )
    )) {isSetParameters=true, module=SetVariables_#5}
  ]
  rendererClasses [&NiagaraMeshRendererProperties]
  renderers [{rendererClass=&NiagaraMeshRendererProperties, rendererIndex=0}]
 }
 StarLineToon {
  emitterSpawnScript [
   /Game/Effects/Niagara/MasterMaterials/ToonShader/Template/NM_ToonShaderParameters(
    Lighting_Hardness=1
    Metallic_Reflection_Color:LinearColor=<rgba 0.6823 0.2439 0 1>
    RimLight_Color:LinearColor=<rgba 1 0.437 0 1>
    RimLight_Width=0.2
   )
   /Game/Effects/Niagara/MasterMaterials/ToonShader/Template/NM_OutlineParameters(OutlineColor:LinearColor=<rgba 1 0.6742 0.026 1>)
  ]
  emitterUpdateScript [
   &EmitterState()
   &SpawnRate(Spawn Probability=&RandomRangeFloat(Minimum=0.5, Maximum=0.7), SpawnRate=4, Use Spawn Probability=true)
  ]
  particleSpawnScript [
   &FN_SetSpawnGroup(Number of Spawn Groups=2)
   &InitializeParticle(
    Color Mode=::NewEnumerator0
    Lifetime=0.6
    Mesh Scale=&Multiply_VectorByFloat(Vector=<1 1 1>, Float=&RandomRangeFloat(Minimum=0.8, Maximum=1))
    Mesh Scale Mode=::NewEnumerator4
    UsePositionOffset=false
    Write Material Random=true
   )
   &MeshRotationForce(Rotation=<2 2 2>)
   &InitialMeshOrientation(Use Rotation Vector=false)
   &ApplyInitialForces()
  ]
  particleUpdateScript [
   &ParticleState()
   &SkeletalMeshLocation(
    Bone / Socket Index=&SelectIntFromArray(Direct Array Index=$Particles.SpawnGroup, Int Selection Array=DataInterface() {propertyValues {IntData [0,1]}})
    Bone Sampling Mode=::NewEnumerator4
    MaskSpawnGroup=false
    Skeletal Mesh=$User.SkeletalMesh
   )
   &ScaleMeshSize(Scale Factor=&Multiply_VectorByFloat(
     Vector=<1 1 1>
     Float=&FloatFromCurve(FloatCurve=DataInterface() {propertyValues {Curve {keys [
          {interpMode=RCIM_Cubic, tangentMode=RCTM_User, time=0, value=0}
          <at 0.2 1>
          {interpMode=RCIM_Cubic, tangentMode=RCTM_User, time=0.7, value=1}
          {tangentMode=RCTM_User, time=1, value=0}
         ]}}})
    ))
   &Color(Color=<rgba 1 0.7924 0 1>)
   /SetVariables(Particles.InterpPos:NiagaraPosition=&AddVectorToPosition(
     Position=$Particles.SkeletalMeshLocation.SampledPosition
     Vector=&TransformVector(
      Vector=&Lerp_Vector(
       A=&Multiply_VectorByFloat(
        Vector=&NormalizeVector(Vector To Normalize=$System.SkydiveVelocity, Safe Normalize=true, Safe Fallback Threshold=0, Safe Fallback Vector=<0 0 0.001>)
        Float=1
       )
       B=&Multiply_VectorByFloat(
        Vector=&NormalizeVector(Vector To Normalize=$System.SkydiveVelocity, Safe Normalize=true, Safe Fallback Threshold=0, Safe Fallback Vector=<0 0 0.001>)
        Float=&RandomRangeFloat(Minimum=-100, Maximum=-150)
       )
       Alpha=$Particles.Age
      )
      Source Space=::World
      Destination Space=::Simulation
     )
    )) {isSetParameters=true, module=SetVariables_#6}
   &InterpolateOverTime(
    Rate Of Change:NiagaraFloat=&Lerp_Float(A=18, B=15, Alpha=$Particles.NormalizedAge)
    Target Vector Value:NiagaraPosition=$Particles.InterpPos
    Variable Type=::NewEnumerator2
   ) {module=InterpolateOverTime001}
   /SetVariables(Particles.Position=$Particles.InterpolateOverTime001.InterpolateOverTime_V3_.Moving Average) {isSetParameters=true, module=SetVariables_#7}
   /SetVariables(Particles.StarScale:Vector3f=&Multiply_VectorByFloat(Vector=$Particles.Scale, Float=0.1)) {isSetParameters=true, module=SetVariables_#8}
  ]
  rendererClasses [&NiagaraMeshRendererProperties]
  renderers [
   {rendererClass=&NiagaraMeshRendererProperties, rendererIndex=0}
   {rendererClass=&NiagaraMeshRendererProperties, rendererIndex=1}
  ]
 }
 StarToon {
  emitterSpawnScript [
   &NM_ToonShaderParameters(
    Lighting_Hardness=1
    Metallic_LightingPower=0
    Metallic_Reflection_Color:LinearColor=<rgba 0 1 0.4868 1>
    RimLight_Color:LinearColor=<rgba 0.0377 2 0 1>
    RimLight_Width=0
   )
   &NM_ToonShaderParameters(
    Lighting_Hardness=1
    Metallic_LightingPower=0
    Metallic_Reflection_Color:LinearColor=<rgba 1 0.3125 0.8693 1>
    RimLight_Color:LinearColor=<rgba 2 0 1.8443 1>
    RimLight_Width=0
   ) {module=NM_ToonShaderParameters002}
   &NM_ToonShaderParameters(
    Lighting_Hardness=1
    Metallic_LightingPower=0
    Metallic_Reflection_Color:LinearColor=<rgba 1 0.7823 0 1>
    RimLight_Color:LinearColor=<rgba 2 1.1268 0 1>
    RimLight_Width=0
   ) {module=NM_ToonShaderParameters001}
   &NM_OutlineParameters(OutlineColor:LinearColor=<rgba 0 0.6458 0.0139 1>)
   &NM_OutlineParameters(OutlineColor:LinearColor=<rgba 0.6198 0 0.348 1>) {module=NM_OutlineParameters002}
   &NM_OutlineParameters(OutlineColor:LinearColor=<rgba 0 0.3465 1 1>, OutlineWidth=3) {module=NM_OutlineParameters001}
  ]
  emitterUpdateScript [
   &EmitterState(Delay First Loop Only=true, Life Cycle Mode=::NewEnumerator1, Loop Delay=0.2, Loop Duration=1, UseLoopDelay=true)
   &SpawnRate(
    Spawn Probability=&RandomRangeFloat(Minimum=0.66, Maximum=0.77)
    SpawnRate=18
    Use Spawn Probability=true
   )
  ]
  particleSpawnScript [
   &PartitionParticles()
   &InitializeParticle(
    Color Mode=::NewEnumerator0
    Lifetime=0.6
    Mesh Scale=&Multiply_VectorByFloat(Vector=<1 1 1>, Float=&RandomRangeFloat(Minimum=0.5, Maximum=0.7))
    Mesh Scale Mode=::NewEnumerator4
    UsePositionOffset=false
    Write Material Random=true
   )
   &MeshRotationForce(Rotation=<2 2 2>)
   &InitialMeshOrientation(Use Rotation Vector=false)
   /SetVariables(Particles.VisibilityTag=&RandomRangeInt(Minimum=0, Maximum=2)) {isSetParameters=true}
   &Color(Color=&SelectLinearColorFromArray(
     Array Sampling Mode=::NewEnumerator1
     Direct Array Index=$Particles.VisibilityTag
     Color Selection Array=DataInterface() {propertyValues {ColorData [<rgba 0.1367 1 0.4018 1>, <rgba 1 0.7238 0 1>, <rgba 1 0 0.1432 1>]}}
     Random Seed=?
    ))
   &CameraOffset(Camera Offset Amount=10)
   &ApplyInitialForces()
  ]
  particleUpdateScript [
   &ParticleState()
   &SkeletalMeshLocation(
    Bone / Socket Index=&SelectIntFromArray(Direct Array Index=$Particles.PartitionParticles.Partition, Int Selection Array=DataInterface() {propertyValues {IntData [0,1,2,3]}})
    Bone Sampling Mode=::NewEnumerator4
    MaskSpawnGroup=false
    Skeletal Mesh=$User.SkeletalMesh
   )
   &ScaleMeshSize(Scale Factor=&Multiply_VectorByFloat(
     Vector=<1 1 1>
     Float=&FloatFromCurve(FloatCurve=DataInterface() {propertyValues {Curve {keys [
          {interpMode=RCIM_Cubic, tangentMode=RCTM_User, time=0, value=0}
          <at 0.2 1>
          {interpMode=RCIM_Cubic, tangentMode=RCTM_User, time=0.7, value=1}
          {tangentMode=RCTM_User, time=1, value=0}
         ]}}})
    ))
   /SetVariables(Particles.InterpPos:NiagaraPosition=&AddVectorToPosition(
     Position=&AddVectorToPosition(
      Position=$Particles.SkeletalMeshLocation.SampledPosition
      Vector=&TransformVector(
       Vector=&RandomRangeVector(Minimum=<-7 -7 -7>, Maximum=<7 7 7>)
       Destination Space=::Simulation
       Apply Scale=true
      )
     )
     Vector=&TransformVector(
      Vector=&Lerp_Vector(
       A=&Multiply_VectorByFloat(
        Vector=&NormalizeVector(Vector To Normalize=$System.SkydiveVelocity, Safe Normalize=true, Safe Fallback Threshold=0, Safe Fallback Vector=<0 0 0.001>)
        Float=1
       )
       B=&Multiply_VectorByFloat(
        Vector=&NormalizeVector(Vector To Normalize=$System.SkydiveVelocity, Safe Normalize=true, Safe Fallback Threshold=0, Safe Fallback Vector=<0 0 0.001>)
        Float=&RandomRangeFloat(Minimum=-75, Maximum=-100)
       )
       Alpha=$Particles.Age
      )
      Source Space=::World
      Destination Space=::Simulation
     )
    )) {isSetParameters=true, module=SetVariables_#6}
   &InterpolateOverTime(
    Rate Of Change:NiagaraFloat=&Lerp_Float(A=20, B=12, Alpha=$Particles.NormalizedAge)
    Target Vector Value:NiagaraPosition=$Particles.InterpPos
    Variable Type=::NewEnumerator2
   )
   /SetVariables(Particles.Position:Vector3f=$Particles.InterpolateOverTime.InterpolateOverTime_V3_.Moving Average) {isSetParameters=true, module=SetVariables_#7}
   /SetVariables(Particles.StarScale:Vector3f=&Multiply_VectorByFloat(Vector=$Particles.Scale, Float=0.1)) {isSetParameters=true, module=SetVariables_#8}
   &UpdateMeshOrientation(Rotation Rate=2, Rotation Vector=<1 1 1>)
   &SolveForcesAndVelocity(Clamp Velocity=false, Limit Acceleration=false)
  ]
  rendererClasses [&NiagaraMeshRendererProperties]
  renderers [
   {rendererClass=&NiagaraMeshRendererProperties, rendererIndex=0}
   {rendererClass=&NiagaraMeshRendererProperties, rendererIndex=1}
   {rendererClass=&NiagaraMeshRendererProperties, rendererIndex=2}
   {rendererClass=&NiagaraMeshRendererProperties, rendererIndex=3}
   {rendererClass=&NiagaraMeshRendererProperties, rendererIndex=4}
   {rendererClass=&NiagaraMeshRendererProperties, rendererIndex=5}
  ]
 }
 TigerVisualiser {
  emitterUpdateScript [
   &EmitterState(Life Cycle Mode=::NewEnumerator1, Loop Behavior=::NewEnumerator1, Loop Duration Mode=::NewEnumerator1, UseLoopDelay=false)
   &SpawnBurst_Instantaneous(Use Spawn Probability=false, UseLoopCountLimit=false)
  ]
  particleSpawnScript [
   &InitializeParticle(
    Color=<rgba 1 0.2136 0 1>
    Lifetime=10
    Mesh Scale=<2 2 2>
    Mesh Scale Mode=::NewEnumerator4
    Position Offset=<0 0 100>
    Sprite Size Mode=::NewEnumerator3
    Uniform Sprite Size=250
    UsePositionOffset=true
    Write Material Random=true
   )
   &InitialMeshOrientation(
    Mesh Orientation Mode=::NewEnumerator2
    Orientation Coordinate Space=::Local
    Orientation Vector=<0 0 1>
    Rotation Coordinate Space=::NewEnumerator2
    Use Rotation Vector=false
   )
  ]
  particleUpdateScript [
   &ParticleState(Kill Particles When Lifetime Has Elapsed=false, Let Infinitely Lived Particles Die When Emitter Deactivates=true)
   &ScaleMeshSize(Scale Factor=&MakeVector(X=1, Y=1, Z=&Lerp_Float(A=0.25, B=1.25, Alpha=$System.Timeline.Float)))
   &SkeletalMeshLocation(
    Bone / Socket Index=4
    Bone Sampling Mode=::NewEnumerator4
    MaskSpawnGroup=false
    Orientation Sampling=::NewEnumerator1
    Sampled Position Offset=<0 0 90>
    Skeletal Mesh=$User.Skeletal Mesh
   )
   /SetVariables(Particles.MeshOrientation=&OrientMeshToVector(
     Look At Direction=&FN_FrontEndVector3Modifier(
      Gameplay=&Lerp_Vector(
       A=&RotateVectorByQuaternion(Vector To Rotate=$Engine.Owner.Velocity, Quaternion=/Niagara/DynamicInputs/Rotation/InvertQuaternion(Quat:Quat4f=$Engine.Owner.Rotation))
       B=<0 0 1>
       Alpha=$System.Dive
      )
      Front End=&Lerp_Vector(A=<0 0 -1>, B=<0 2 -1>, Alpha=$System.Dive)
     )
     Reference Axis Vector=<1 0 0>
    )) {isSetParameters=true}
   &DynamicMaterialParameters(
    Index 0 Param 1=&Lerp_Float(A=0.3, B=0.75, Alpha=$User.ParaGlideLeanAlpha)
    Index 0 Param 2=&Add_Float(A=0.1, B=$Emitter.Age)
    Index 0 Param 3=$System.Timeline.Float
    Index 0 Param 4=&Lerp_Float(A=&Multiply_Float(A=0.9, B=$System.Timeline001.Float), B=0.5, Alpha=$System.Dive)
    Module.Param0WriteEnabled=true
    Module.Param1WriteEnabled=true
    Module.Param2WriteEnabled=true
    Module.Param3WriteEnabled=true
   )
   &SolveForcesAndVelocity(Clamp Velocity=false, Limit Acceleration=false)
  ]
  rendererClasses [&NiagaraMeshRendererProperties]
  renderers [
   {rendererClass=&NiagaraMeshRendererProperties, rendererIndex=0}
   {rendererClass=&NiagaraMeshRendererProperties, rendererIndex=1}
  ]
 }
}
summary {
 systemName=NS_Skydive_Dazzle_v4
 emitters [
  {emitterName=TigerVisualiser, rendererClasses [&NiagaraMeshRendererProperties]}
  {emitterName=StarToon, rendererClasses [&NiagaraMeshRendererProperties]}
  {emitterName=StarLineToon, rendererClasses [&NiagaraMeshRendererProperties]}
  {emitterName=Arrow, rendererClasses [&NiagaraMeshRendererProperties]}
  {emitterName=Ring, rendererClasses [&NiagaraMeshRendererProperties]}
  {emitterName=Ribbons, rendererClasses [&NiagaraRibbonRendererProperties]}
  {emitterName=Squiggle, rendererClasses [&NiagaraMeshRendererProperties]}
 ]
 userVariables [
  User.bIsFrontEnd {type=NiagaraBool}
  User.bIsFrontEndPreview {type=NiagaraBool, defaultValue=true}
  User.LocalHorizontalDistance {type=NiagaraFloat, defaultValue=0, description=""}
  User.LocalHorizontalVector {type=Vector3f, description="", defaultValue {struct=Vector3f, value=<0 0 0>}}
  User.MeshColors {
   type=NiagaraDataInterfaceArrayColor
   description=""
   defaultValue {
    struct=NiagaraExt_VariableValue_DataInterface
    value {
     dataInterface=&self:NiagaraDataInterfaceArrayColor_0
     dataInterfaceClass=&NiagaraDataInterfaceArrayColor
    }
   }
  }
  User.ParaGlideLeanAlpha {type=NiagaraFloat, defaultValue=0, description=""}
  User.PivotOffset {type=NiagaraFloat, defaultValue=0, description=""}
  User.PlayerSpeed {type=NiagaraFloat, defaultValue=0, description=""}
  User.RibbonIDOffset {type=NiagaraInt32}
  User.Skeletal Mesh {
   type=NiagaraDataInterfaceSkeletalMesh
   defaultValue {
    struct=NiagaraExt_VariableValue_DataInterface
    value {dataInterface=&self:NiagaraDataInterfaceSkeletalMesh_1}
   }
  }
  User.SkeletalMesh {
   type=NiagaraDataInterfaceSkeletalMesh
   defaultValue {
    struct=NiagaraExt_VariableValue_DataInterface
    value {dataInterface=&self:NiagaraDataInterfaceSkeletalMesh_0}
   }
  }
  User.TimeOffset {type=NiagaraFloat, defaultValue=0, description=""}
 ]
}
