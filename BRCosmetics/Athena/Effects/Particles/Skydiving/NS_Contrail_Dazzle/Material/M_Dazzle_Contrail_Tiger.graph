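# Translucent unlit material for Niagara mesh particles in the Dazzle skydiving contrail's Tiger variant.
# Polar-UV Perlin noise warps a radial sphere mask, with dynamic parameters driving distortion strength, animation time and opening.
# The mask yields animated sine rings tinting a tiger texture, plus a noise-broken pink center, summed for emissive.
# A separate time-based sine, scaled by WPOStrenght, builds a Z-axis wobble offset.
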Material /BRCosmetics/Athena/Effects/Particles/Skydiving/NS_Contrail_Dazzle/Material/M_Dazzle_Contrail_Tiger
graph {
 properties {
  bUsedWithNiagaraMeshParticles=true
  blendMode=BLEND_Translucent
  pixelDepthOffsetMode=PDOM_Legacy
  shadingModel=MSM_Unlit
  usedShadingModels=Unlit
 }
 nodes [
  n1 = TextureCoordinate()
  n2 = Constant2Vector() {desc=Offset, g=-0.03}
  n3 = TextureSample(UVs=MaterialFunctionCall(UVs=Add(A=@n1, B=@n2), Texture Scale=Constant2Vector() {desc=Scale, g=0.72, r=0.7}) {materialFunction=&ScaleUVsByCenter}.UVs) {texture=/BRCosmetics/Athena/Effects/Particles/Skydiving/NS_Contrail_Dazzle/Textures/T_Dazzle_Tiger_D}
  n4 = DynamicParameter() {paramNames [AmplitudeStrenght, Time, Opening, WPOStrenght]}
  n5 = Multiply(A=@n4.Time) {constB=4}
  n6 = Constant2Vector() {g=1, r=1}
  n7 = Constant()
  n8 = MaterialFunctionCall!(UV=TextureCoordinate(), Twist Center=@n7, Twist Edges=@n7) {materialFunction=/Game/Accessories/FORT_Backpacks/Backpack_F_MED_Celestial_Glow/FX/Materials/MF_PolarUV}
  n9 = Constant2Vector() {g=0.1, r=2}
  n10 = TextureSample!(
   UVs=Panner(Coordinate=Multiply(A=@n8.Polar UVs, B=@n9), Time=@n4.Time) {speedY=-0.5}
   Tex
   "DDX(UVs)"=@n8.DDX
   "DDY(UVs)"=@n8.DDX
   Apply View MipBias
  ) {
   mipValueMode=TMVM_Derivative
   samplerType=SAMPLERTYPE_Normal
   texture=/BRCosmetics/Characters/Player/Male/Large/Bodies/M_LRG_Clash_V/Textures/Simple_Perlin_N
  }
  n11 = OneMinus(None=SphereMask(A=TextureCoordinate(), B=Constant2Vector() {g=0.5, r=0.5}, Radius=Constant() {r=0.2}, Hardness=Constant() {r=0.5}))
  n12 = Multiply(
   A=MaterialFunctionCall(X=Multiply(A=@n10.R, B=@n11), Y=Multiply(A=@n10.G, B=@n11)) {materialFunction=&MakeFloat2}.Result
   B=LinearInterpolate(Alpha=NamedRerouteDeclaration(None=@n4.AmplitudeStrenght) {name=Strenght, nodeColor=<rgba 0.875 0 1 1>}) {constB=0.5}
  )
  n13 = SphereMask(
   A=MaterialFunctionCall(UVs=TextureCoordinate(), Texture Scale=Add(A=@n6, B=@n12)) {materialFunction=&ScaleUVsByCenter}.UVs
   B=Constant2Vector() {g=0.5, r=0.5}
   Radius=LinearInterpolate(Alpha=DynamicParameter() {paramNames [AmplitudeStrenght, Time, Opening, WPOStrenght]}.Opening) {constB=0.4}
   Hardness=Constant()
  )
  n14 = Multiply(A=@n13, B=ScalarParameter() {defaultValue=15, parameterName=LinesAmount}) {constB=30}
  n15 = Power(Base=Sine(None=Add(A=@n5, B=@n14)), Exp=ScalarParameter() {defaultValue=1, parameterName=LineWidth}) {constExponent=0.5}
  n16 = OneMinus(None=SmoothStep(Value=@n13) {constMax=0.725, constMin=0.7})
  n17 = SmoothStep(Value=@n13) {constMax=0.8, constMin=0.75}
  n18 = Add(A=OneMinus(None=SmoothStep(Value=@n13) {constMin=0.6})) {constB=0.5}
  n19 = Multiply(A=Add(A=Multiply(A=@n15, B=@n16), B=Multiply(A=@n17, B=@n18)), B=SmoothStep(Value=@n13) {constMax=0.325, constMin=0.325})
  n20 = Constant3Vector() {constant=<rgba 0.7758 0.0844 0.2918 1>}
  n21 = SmoothStep(Value=@n13) {constMax=0.3, constMin=0.25}
  n22 = Constant()
  n23 = MaterialFunctionCall(UV=TextureCoordinate(), Twist Center=@n22, Twist Edges=@n22) {materialFunction=&MF_PolarUV}
  n24 = Multiply(A=@n23.Polar UVs, B=Constant2Vector() {r=12})
  n25 = Multiply(A=@n23.Polar UVs, B=Constant2Vector() {r=6})
  n26 = Multiply(
   A=SmoothStep(Value=TextureSample!(UVs=@n24, Tex, Apply View MipBias) {texture=/Game/Global/Textures/Default/Noise/3xPerlin_M}.R) {constMax=0.3, constMin=0.15}
   B=SmoothStep(Value=TextureSample(UVs=@n25) {texture=&3xPerlin_M}.G) {constMax=0.3, constMin=0.15}
  )
  n27 = LinearInterpolate(Alpha=NamedRerouteUsage()) {constB=0.75}
  n28 = Multiply(A=Multiply(A=OneMinus(None=@n21), B=SmoothStep(Value=@n13) {constMax=0.3, constMin=0.05}), B=OneMinus(None=Multiply(A=@n26, B=@n27)))
  MaterialFunctionCall(Emissive=Add(A=Multiply(A=@n3.RGB, B=@n19), B=Multiply(A=@n20, B=@n28))) {materialFunction=&MF_FX_EmissiveAdjustments}
  MaterialFunctionCall!(MatrixRow3, MatrixRow4, MatrixRow2, Vector, MatrixRow1) {materialFunction=/Engine/Functions/Engine_MaterialFunctions02/Math/MulM4V4}
  n29 = ScalarParameter() {defaultValue=8, parameterName=WPOStrenght}
  n30 = DynamicParameter() {paramNames [AmplitudeStrenght, Time, Opening, WPOStrenght]}
  n31 = Constant() {r=0.5}
  n32 = Multiply(A=SphereMask(A=TextureCoordinate(), B=@n31, Radius=@n31, Hardness=Constant())) {constB=1.5}
  n33 = Multiply(A=Time(), B=ScalarParameter() {defaultValue=2, parameterName=TimeMult}) {constB=2}
  Multiply(
   A=Multiply(A=Constant3Vector() {constant=<rgba 0 0 1 0>}, B=LinearInterpolate(A=@n29, Alpha=@n30.WPOStrenght) {constA=5, constB=0}) {constB=5}
   B=Sine(None=Add(A=@n32, B=@n33))
  )
  Multiply(A=Add(A=@n19, B=@n28))
 ]
}
