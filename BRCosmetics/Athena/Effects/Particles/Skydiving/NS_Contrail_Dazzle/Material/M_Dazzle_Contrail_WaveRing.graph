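# Translucent, unlit, two-sided material for the wave ring in the Dazzle skydiving contrail, enabled for Niagara ribbons and mesh particles.
# Two time-animated wave UV layers drive a sphere-style warp that distorts panning Gaussian noise and shapes a reflected gradient.
# That gradient erodes the contrast-adjusted noise, which modulates a dimmed-to-full particle color blend before emissive adjustment.
# SplineThicken takes its width from a dynamic parameter.
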Material /BRCosmetics/Athena/Effects/Particles/Skydiving/NS_Contrail_Dazzle/Material/M_Dazzle_Contrail_WaveRing
graph {
 properties {
  bCastRayTracedShadows=false
  bUsedWithNiagaraMeshParticles=true
  bUsedWithNiagaraRibbons=true
  blendMode=BLEND_Translucent
  opacityMaskClipValue=0.0001
  pixelDepthOffsetMode=PDOM_Legacy
  shadingModel=MSM_Unlit
  twoSided=true
  usedShadingModels=Unlit
 }
 nodes [
  n1 = DynamicParameter() {parameterIndex=1, defaultValue=<rgba 10 1 1 1>, paramNames [Sline,Param2,Param3,Param4]}
  MaterialFunctionCall(WidthBase=@n1.Sline, WidthTip=@n1.Sline) {materialFunction=&SplineThicken}
  n2 = ParticleColor()
  n3 = Multiply(A=@n2.RGB) {constB=0.45}
  n4 = Multiply(A=TextureCoordinate(), B=Constant2Vector() {g=1, r=0.5})
  n5 = TextureCoordinate()
  n6 = AppendVector(A=ComponentMask(None=@n5) {r=true}, B=ComponentMask(None=@n5) {g=true})
  n7 = OneMinus(None=Time())
  n8 = Multiply(A=@n7) {constB=1.2}
  n9 = MaterialFunctionCall!(UV=@n6, U Frequency, V Frequency=Constant() {r=5}, U Amplitude, V Amplitude=Constant() {r=0.1}, U Speed, V Speed, U Time=@n8, V Time=@n8) {materialFunction=/Game/Effects/Fort_Effects/MaterialFunctions/MF_WaveUV}
  n10 = Multiply(A=@n7) {constB=1.5}
  n11 = MaterialFunctionCall(UV=@n6, V Frequency=Constant() {r=2}, V Amplitude=Constant() {r=0.2}, U Time=@n10, V Time=@n10) {materialFunction=&MF_WaveUV}
  n12 = Saturate(None=Divide(A=Multiply(A=@n9.Wave UVs, B=@n11.Wave UVs)) {constB=0.5})
  n13 = TextureCoordinate() {vTiling=0}
  n14 = Constant2Vector() {r=0.5}
  n15 = Constant() {r=0.5}
  n16 = Constant()
  n17 = Divide(
   A=ConstantBiasScale(None=ComponentMask(None=@n12) {r=true}) {bias=-0.5, scale=2}
   B=Multiply(A=SphereMask(A=@n13, B=@n14, Radius=@n15, Hardness=@n16))
  )
  n18 = OneMinus(None=Saturate(None=DotProduct(A=@n17, B=@n17)))
  n19 = ConstantBiasScale(None=Divide(A=@n17, B=SquareRoot(None=@n18)))
  n20 = TextureSample(UVs=Panner(Coordinate=Add(A=@n4, B=@n19) {constB=3}) {speedX=-0.25, speedY=-1}) {texture=/Game/Effects/Fort_Effects/Textures/Tile/T_GaussianNoise_01_M}
  n21 = Multiply(
   A=MaterialFunctionCall!(Gradient=@n18, Center Point=Constant() {r=1}) {materialFunction=/Game/Effects/Fort_Effects/MaterialFunctions/MF_GradientReflected}.Result
   B=Constant() {r=0.6}
  )
  n22 = Multiply(A=@n20.R, B=@n21)
  n23 = Saturate(None=Multiply(A=@n22) {constB=4})
  n24 = Power(Base=@n20.R, Exp=ScalarParameter() {defaultValue=1.25, parameterName=LTrailContrast}) {constExponent=1.25}
  n25 = ScalarParameter() {defaultValue=1.75, parameterName=LTrailStrenght}
  n26 = OneMinus(None=MaterialFunctionCall!(Gradient=Multiply(A=@n24, B=@n25) {constB=2}, Erosion Amount=@n21, Softness=Constant() {r=0.25}) {materialFunction=/BRCosmetics/Weapons/FORT_Melee/Pickaxe_Turtleneck_Male/FX/Default/MF_Erosion}.Increased Range)
  n27 = Add(A=@n22) {constB=0.7}
  MaterialFunctionCall(Emissive=Multiply(A=LinearInterpolate(A=@n3, B=@n2.RGB, Alpha=@n23), B=Multiply(A=@n26, B=@n27))) {materialFunction=&MF_FX_EmissiveAdjustments}
 ]
}
