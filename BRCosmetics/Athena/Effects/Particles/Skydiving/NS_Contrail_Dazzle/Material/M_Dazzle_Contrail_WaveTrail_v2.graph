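# Unlit translucent material for the Dazzle skydiving contrail that draws pink and blue wavy strands, usable on Niagara ribbons and mesh particles.
# MF_WaveUV bends each strand, with dynamic parameters driving frequency, noise-modulated amplitude, thickness and time.
# A square-root profile across each strand warps panning noise for a rounded look, then erosion and sine bands shade it.
# Opacity combines both strands with TrailFade and tail falloff, plus quality-switched depth fade.
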Material /BRCosmetics/Athena/Effects/Particles/Skydiving/NS_Contrail_Dazzle/Material/M_Dazzle_Contrail_WaveTrail_v2
graph {
 properties {
  bCastRayTracedShadows=false
  bUsedWithNiagaraMeshParticles=true
  bUsedWithNiagaraRibbons=true
  blendMode=BLEND_Translucent
  opacityMaskClipValue=0.0001
  pixelDepthOffsetMode=PDOM_Legacy
  shadingModel=MSM_Unlit
  usedShadingModels=Unlit
 }
 nodes [
  n1 = Constant3Vector() {constant=<rgba 0.7758 0.0844 0.2918 1>}
  n2 = Multiply(A=TextureCoordinate(), B=Constant2Vector() {g=0.5, r=3})
  n3 = NamedRerouteUsage()
  n4 = MaterialFunctionCall!(
   UV=NamedRerouteUsage()
   U Frequency
   V Frequency=NamedRerouteUsage()
   U Amplitude
   V Amplitude=NamedRerouteUsage()
   U Speed
   V Speed
   U Time=@n3
   V Time=@n3
  ) {materialFunction=/Game/Effects/Fort_Effects/MaterialFunctions/MF_WaveUV}
  n5 = TextureCoordinate()
  n6 = NamedRerouteDeclaration(None=OneMinus(None=ComponentMask(None=@n5) {r=true})) {name=DIrectionMask, nodeColor=<rgba 0 1 0.0859 1>}
  n7 = Multiply(
   A=Multiply(A=Multiply(A=@n6, B=@n6), B=ScalarParameter() {defaultValue=1, parameterName=TrailFade})
   B=Add(A=SmoothStep(Value=@n6) {constMax=0.75, constMin=1}, B=ScalarParameter() {defaultValue=0.1, parameterName=BottomWidth}) {constB=0.2}
  )
  n8 = Divide(A=ConstantBiasScale(None=ComponentMask(None=@n4.Wave UVs) {r=true}) {bias=-0.5, scale=2}, B=Multiply(A=@n7, B=NamedRerouteUsage()))
  n9 = OneMinus(None=Saturate(None=DotProduct(A=@n8, B=@n8)))
  n10 = SquareRoot(None=@n9)
  n11 = ConstantBiasScale(None=Divide(A=@n8, B=Saturate(None=@n10)))
  n12 = TextureSample(UVs=Panner(Coordinate=Add(A=@n2, B=@n11) {constB=3}) {speedX=-3, speedY=-0.25}) {texture=/Game/Effects/Fort_Effects/Textures/Tile/T_GaussianNoise_01_M}
  n13 = Multiply(
   A=Power(Base=@n12.R, Exp=ScalarParameter() {defaultValue=1.25, parameterName=LTrailContrast}) {constExponent=1.25}
   B=ScalarParameter() {defaultValue=2, parameterName=LTrailStrenght}
  ) {constB=2}
  n14 = Constant() {r=0.25}
  n15 = Multiply(
   A=OneMinus(None=MaterialFunctionCall!(Gradient=@n13, Erosion Amount=@n9, Softness=@n14) {materialFunction=/BRCosmetics/Weapons/FORT_Melee/Pickaxe_Turtleneck_Male/FX/Default/MF_Erosion}.Increased Range)
   B=Add(A=Multiply(A=@n12.R, B=@n9)) {constB=0.7}
  )
  n16 = LinearInterpolate(A=LinearInterpolate(A=@n1, B=Multiply(A=@n1) {constB=0.6}, Alpha=@n15), B=Multiply(A=@n1) {constB=0.3}, Alpha=OneMinus(None=@n9))
  n17 = TextureCoordinate()
  n18 = Add(A=NamedRerouteUsage(), B=Add(A=ComponentMask(None=@n17) {r=true}, B=Multiply(A=@n12.R) {constB=-0.25}))
  n19 = LinearInterpolate(A=@n16, B=Multiply(A=@n16) {constB=0.7}, Alpha=Multiply(A=@n9, B=Sine(None=@n18)))
  n20 = Constant3Vector() {constant=<rgba 0 0.45 1 1>}
  n21 = Multiply(A=TextureCoordinate(), B=Constant2Vector() {g=0.25, r=2})
  n22 = NamedRerouteUsage()
  n23 = MaterialFunctionCall(UV=NamedRerouteUsage(), V Frequency=NamedRerouteUsage(), V Amplitude=NamedRerouteUsage(), U Time=@n22, V Time=Add(A=@n22) {constB=0.5}) {materialFunction=&MF_WaveUV}
  n24 = Divide(A=ConstantBiasScale(None=ComponentMask(None=@n23.Wave UVs) {r=true}) {bias=-0.5, scale=2}, B=Multiply(A=@n7, B=NamedRerouteUsage()))
  n25 = OneMinus(None=Saturate(None=DotProduct(A=@n24, B=@n24)))
  n26 = SquareRoot(None=@n25)
  n27 = ConstantBiasScale(None=Divide(A=@n24, B=Saturate(None=@n26))) {scale=0.25}
  n28 = TextureSample(UVs=Panner(Coordinate=Add(A=@n21, B=@n27) {constB=3}) {speedX=-3, speedY=-0.25}) {texture=&T_GaussianNoise_01_M}
  n29 = Power(Base=@n28.R, Exp=ScalarParameter() {defaultValue=1.25, parameterName=LTrailContrast}) {constExponent=1.25}
  n30 = ScalarParameter() {defaultValue=2, parameterName=LTrailStrenght}
  n31 = OneMinus(None=MaterialFunctionCall(Gradient=Multiply(A=@n29, B=@n30) {constB=2}, Erosion Amount=@n25, Softness=Constant() {r=0.25}) {materialFunction=&MF_Erosion}.Increased Range)
  n32 = Multiply(A=@n31, B=Add(A=Multiply(A=@n28.R, B=@n25)) {constB=0.7})
  n33 = LinearInterpolate(A=LinearInterpolate(A=@n20, B=Multiply(A=@n20) {constB=0.6}, Alpha=@n32), B=Multiply(A=@n20) {constB=0.4}, Alpha=OneMinus(None=@n25))
  n34 = Add(A=Add(A=NamedRerouteUsage()) {constB=0.5}, B=ComponentMask(None=TextureCoordinate()) {r=true})
  n35 = LinearInterpolate(A=@n33, B=Multiply(A=@n33) {constB=0.7}, Alpha=Multiply(A=@n25, B=Sine(None=@n34)))
  MaterialFunctionCall(Emissive=LinearInterpolate(A=Saturate(None=@n19), B=Saturate(None=@n35), Alpha=Ceil(None=@n31))) {materialFunction=&MF_FX_EmissiveAdjustments}
  n36 = Max(A=@n15, B=@n32)
  n37 = OneMinus(None=ComponentMask(None=TextureCoordinate()) {r=true})
  n38 = Saturate(None=Add(A=Power(Base=@n37), B=Multiply(A=@n12.R) {constB=0.25}))
  n39 = OneMinus(None=ComponentMask(None=TextureCoordinate()) {r=true})
  n40 = Saturate(None=Add(A=Power(Base=@n39), B=Multiply(A=@n28.R) {constB=0.25}))
  n41 = LinearInterpolate(A=Multiply(A=Multiply(A=@n15, B=@n38)) {constB=0.35}, B=Multiply(A=Multiply(A=@n32, B=@n40)) {constB=0.35}, Alpha=@n32)
  n42 = Power(Base=OneMinus(None=ScalarParameter() {defaultValue=1, parameterName=TrailFade})) {constExponent=0.5}
  n43 = DepthFade() {fadeDistanceDefault=15}
  n44 = Constant() {r=1}
  Multiply(
   A=SmoothStep(Value=LinearInterpolate(A=@n36, B=@n41, Alpha=@n42) {constB=0}) {constMax=0.55, constMin=0.45}
   B=QualitySwitch(Default=@n43, Low=@n44, High=@n43, Medium=@n44, Epic=@n43)
  )
  n45 = OneMinus(None=Time())
  n46 = DynamicParameter() {defaultValue=<rgba 2.3 1.5 1 1>, paramNames [WaveFrequencyL, WaveFrequencyR, TrailTImeMult, TrailTimeAdd]}
  NamedRerouteDeclaration(None=Add(A=Multiply(A=@n45, B=@n46.TrailTImeMult) {constB=2}, B=@n46.TrailTimeAdd)) {name=Time, nodeColor=<rgba 1 0.0469 0 1>}
  NamedRerouteDeclaration(None=@n46.WaveFrequencyL) {name=WaveFrequencyL, nodeColor=<rgba 0.4766 1 0 1>}
  NamedRerouteDeclaration(None=@n46.WaveFrequencyR) {name=WaveFrequencyR, nodeColor=<rgba 0 1 0.625 1>}
  n47 = DynamicParameter() {parameterIndex=1, defaultValue=<rgba 0.25 0.25 0.25 1>, paramNames [AmplitudeNoiseSpeed, AmplitudeContrast, AmplitudeStrenght, TrailTimeAdd]}
  n48 = Divide(A=@n47.AmplitudeNoiseSpeed) {constB=4}
  n49 = TextureSample(UVs=Panner(Coordinate=TextureCoordinate() {uTiling=0.25, vTiling=0}, Time=NamedRerouteUsage(), Speed=AppendVector(A=@n47.AmplitudeNoiseSpeed, B=@n48)) {speedX=0.25}) {texture=&T_GaussianNoise_01_M}
  NamedRerouteDeclaration(None=Multiply(A=Add(A=@n49.RGB, B=@n47.AmplitudeContrast) {constB=0.5}, B=@n47.AmplitudeStrenght)) {name=AmplitudeL, nodeColor=<rgba 0.9688 0 1 1>}
  n50 = TextureCoordinate()
  n51 = ComponentMask(None=@n50) {g=true}
  n52 = ComponentMask(None=@n50) {r=true}
  n53 = Constant2Vector() {g=1, r=1}
  n54 = Constant2Vector() {g=0.75, r=0.75}
  n55 = OneMinus(None=SmoothStep(Value=@n52) {constMax=0.15})
  NamedRerouteDeclaration(None=MaterialFunctionCall(UVs=AppendVector(A=@n51, B=@n52), Texture Scale=LinearInterpolate(A=@n53, B=@n54, Alpha=@n55)) {materialFunction=&ScaleUVsByCenter}.UVs) {name=UV, nodeColor=<rgba 0.2891 1 0 1>}
  n56 = DynamicParameter() {parameterIndex=2, defaultValue=<rgba 0.7 0.3 1 1>, paramNames [ThicknessL,ThicknessR,Param3,Param4]}
  NamedRerouteDeclaration(None=@n56.ThicknessL) {name=ThicknessL, nodeColor=<rgba 0.0781 0 1 1>}
  NamedRerouteDeclaration(None=@n56.ThicknessR) {name=ThicknessR, nodeColor=<rgba 0.0781 0 1 1>}
 ]
}
