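# Unlit translucent, two-sided material for Dazzle skydiving contrail Niagara mesh particles, scrolling a graphic (arrow by default) along a spline-thickened strip.
# Dynamic parameters drive scroll position and SplineThicken width; the FrameRate and ScrollAmountPerFrame channels are unconnected.
# Sphere masks define pinch UV scaling and a pinch mask, shaping the cropped, non-repeating scroll texture.
# A hard SmoothStep thresholds it, optionally multiplied by a static zigzag, while particle color feeds emissive.
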Material /BRCosmetics/Athena/Effects/Particles/Skydiving/NS_Contrail_Dazzle/Material/M_Dazzle_GraphicScroll
graph {
 properties {
  bCastRayTracedShadows=false
  bUsedWithNiagaraMeshParticles=true
  blendMode=BLEND_Translucent
  opacityMaskClipValue=0.1
  pixelDepthOffsetMode=PDOM_Legacy
  shadingModel=MSM_Unlit
  twoSided=true
  usedShadingModels=Unlit
 }
 nodes [
  n1 = DynamicParameter() {defaultValue=<rgba 0.5 12 3 10>, paramNames [Scroll, FrameRate, ScrollAmountPerFrame, SplineWidth]}
  MaterialFunctionCall(WidthBase=@n1.SplineWidth, WidthTip=@n1.SplineWidth) {materialFunction=&SplineThicken}
  n2 = ParticleColor()
  n3 = TextureCoordinate()
  n4 = Multiply(
   A=@n3
   B=AppendVector(
    A=ScalarParameter() {defaultValue=1, parameterName=XTileScrollTexture}
    B=ScalarParameter() {defaultValue=4, parameterName=YTileScrollTexture}
   )
  )
  n5 = NamedRerouteUsage()
  n6 = Constant()
  n7 = LinearInterpolate(
   A=ScalarParameter() {defaultValue=1, parameterName=ScrollLocationEnd}
   B=ScalarParameter() {defaultValue=-1, parameterName=ScrollLocationBeginning}
   Alpha=DynamicParameter() {defaultValue=<rgba 0.5 12 3 10>, paramNames [Scroll, FrameRate, ScrollAmountPerFrame, SplineWidth]}.Scroll
  ) {constA=1.5, constB=-4.5}
  n8 = TextureObjectParameter() {
   parameterName=ScrollTexture
   samplerType=SAMPLERTYPE_Grayscale
   texture=/BRCosmetics/Athena/Effects/Particles/Skydiving/NS_Contrail_Dazzle/Textures/T_Dazzle_Arrow_M
  }
  n9 = MaterialFunctionCall!(
   UVs=Add(
    A=MaterialFunctionCall(UVs=@n4, Texture Scale=@n5) {materialFunction=&ScaleUVsByCenter}.UVs
    B=AppendVector(A=@n6, B=@n7)
   )
   TextureIn=@n8
   UpperLeftCorner=Constant2Vector()
   LowerRightCorner=Constant2Vector() {g=1, r=1}
   ExportFloat4
  ) {materialFunction=/Engine/Functions/Engine_MaterialFunctions01/Texturing/TextureCropping}
  n10 = NamedRerouteUsage()
  n11 = Multiply(A=@n9.Cropped, B=@n10)
  n12 = Multiply(A=@n2.RGB, B=@n11)
  n13 = ScalarParameter() {defaultValue=0.5, parameterName=DiffStrenght}
  MaterialFunctionCall(Emissive=StaticSwitchParameter(True=LinearInterpolate(A=@n12, B=@n2.RGB, Alpha=@n13), False=@n2.RGB) {defaultValue=true, parameterName=UseScollTextureDiff}) {materialFunction=&MF_FX_EmissiveAdjustments}
  n14 = TextureCoordinate() {uTiling=0}
  n15 = Constant2Vector() {g=0.45, r=0.1}
  n16 = Constant() {r=0.45}
  n17 = Constant() {r=0.25}
  NamedRerouteDeclaration(None=AppendVector(A=SphereMask(A=@n14, B=@n15, Radius=@n16, Hardness=@n17), B=Constant() {r=1})) {name=Pinch, nodeColor=<rgba 0 1 0.5313 1>}
  n18 = TextureCoordinate() {vTiling=0}
  n19 = NamedRerouteUsage()
  NamedRerouteDeclaration(None=SphereMask(
    A=MaterialFunctionCall(UVs=@n18, Texture Scale=@n19) {materialFunction=&ScaleUVsByCenter}.UVs
    B=Constant2Vector() {r=0.5}
    Radius=Constant() {r=0.525}
    Hardness=Constant() {r=0.5}
   )) {name=Pinchmask, nodeColor=<rgba 1 0 0.4922 1>}
  n20 = AppendVector(
   A=ScalarParameter() {defaultValue=1, parameterName=XTileStaticTexture}
   B=ScalarParameter() {defaultValue=4, parameterName=YTileStaticTexture}
  )
  n21 = TextureSampleParameter2D(UVs=MaterialFunctionCall(UVs=Multiply(A=@n3, B=@n20), Texture Scale=@n5) {materialFunction=&ScaleUVsByCenter}.UVs) {
   parameterName=StaticTexture
   samplerType=SAMPLERTYPE_Grayscale
   texture=/BRCosmetics/Athena/Effects/Particles/Skydiving/NS_Contrail_Dazzle/Textures/T_Dazzle_Zigzag_A
  }
  n22 = Multiply(A=@n21.R, B=@n9.Cropped)
  SmoothStep(Value=StaticSwitchParameter(True=Multiply(A=@n22, B=@n10), False=@n11) {parameterName=UseStaticTextureWithScroll}) {constMax=0.1, constMin=0.1}
 ]
}
